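; Weapon Blueprint for the dual-wield CarolinaChili pickaxe cosmetic that drives its swing and impact effects.
; Right and left swings pass the montage section index into their VFX components as SwingSectionIndex, as does the in-game impact FX.
; In preview or frontend contexts, impact instead deactivates the default effect and spawns a Niagara system relative to the instigator.
; That preview gets a cycling counter-based section and variant, with DoOnce and a short delay against duplicate signals.
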(event EventSwingRight_Common
  ;[0 0 0] Swing
  (|CallParentFunction)
  (bind _section_index (Utilities|SetupSwingMontageSectionIndex (Variables|Actor|GetInstigator))
    (:Out
      (Effects|Components|ParticleSystem|SetIntParameter (CallFunction|GetSwingVFXComponent) "SwingSectionIndex" _section_index))))

(event EventSwingLeft_Common
  ;[0 0 0] Swing
  (|CallParentFunction)
  (bind _section_index (Utilities|SetupSwingMontageSectionIndex (Variables|Actor|GetInstigator))
    (:Out
      (Effects|Components|ParticleSystem|SetIntParameter (CallFunction|GetSwingOffhandVFXComponent) "SwingSectionIndex" _section_index))))

(event EventSwingLeft
  (|CallParentFunction))

(event EventSwingRight
  (|CallParentFunction))

(event FortWeapon|EventOnPlayImpactFX
  (params (HitResult :type Hit%20Result%20Structure%20%28by%20ref%29) (ImpactPhysicalSurface :type EPhysicalSurface%20Enum :default "SurfaceType_Default") (SpawnedPSC :type FXSystem%20Component%20Object%20Reference))
  (bind _instigator (Variables|Actor|GetInstigator))
  (bind (_returnvalue_location _returnvalue_rotation _returnvalue_scale) (Math|Transform|BreakTransform (Transformation|GetActorTransform _instigator)))
  ;[0 0 0] Impact
  (|CallParentFunction HitResult ImpactPhysicalSurface SpawnedPSC)
  (if (or (Reactive|Frontend|IsCosmeticObjectDisplayedasPreview 0) (Reactive|Frontend|IsCosmeticObjectDisplayedinFrontendContext 0))
    ;[0 0 0] FrontEnd
    (Components|Activation|Deactivate SpawnedPSC)
    (bind _returnvalue (Niagara|SpawnSystematLocation "/BRCosmetics/Weapons/FORT_Melee/Pickaxe_CarolinaChili/FX/NS_Pickaxe_CarolinaChili.NS_Pickaxe_CarolinaChili" (Math|Transform|TransformLocation (Transformation|GetActorTransform _instigator) "100.000000,0.000000,50.000000") (Math|Rotator|CombineRotators _returnvalue_rotation "0.000000,180.000000,0.000000") _returnvalue_scale))
    (Utilities|FlowControl|DoOnce false
      (:Completed
        (Effects|Components|ParticleSystem|SetIntParameter _returnvalue "SwingSectionIndex" (Math|Integer|%25%28Integer%29 (Variables|Default|GetFE_Counter) 4))
        (Effects|Components|ParticleSystem|SetBoolParameter _returnvalue "bIsFrontEndPreview" true)
        (Effects|Components|ParticleSystem|SetFloatParameter _returnvalue "Variant" (Math|Conversions|ToFloat(Integer) (Variables|Default|GetFE_VariantIndex)))
        (Components|Activation|Activate _returnvalue false)
        (Math|Integer|IncrementInt (Variables|Default|GetFE_Counter)
          (:%20%20
            (Utilities|FlowControl|RetriggerableDelay 0.25) ; prevents duplicated signal
          ))))
    (else
      (bind _section_index (Utilities|SetupSwingMontageSectionIndex _instigator)
        (:Out
          (Effects|Components|ParticleSystem|SetIntParameter (Variables|Default|GetImpact%20FX) "SwingSectionIndex" _section_index))))))

(fn ConstructionScript
  (|CallParentFunction))
