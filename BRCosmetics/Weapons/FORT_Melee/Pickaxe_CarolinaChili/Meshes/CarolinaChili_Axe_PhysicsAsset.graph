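# Physics asset for the CarolinaChili pickaxe mesh, a cosmetic two-handed melee weapon.
# It holds a single body on the Melee_TwoHanded_Handle bone, made of one angled capsule offset from the bone origin.
# It has no constraints, so the axe has one simple collision shape with no jointed parts.
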PhysicsAsset /BRCosmetics/Weapons/FORT_Melee/Pickaxe_CarolinaChili/Meshes/CarolinaChili_Axe_PhysicsAsset
bodies {Melee_TwoHanded_Handle [{shapeType=Capsule, length=21.8881, radius=12.0953, center=<13.0979 1.7181 -1.1022>, rotation=<rot 0 4.9415 -73.3376>}]}
constraints []
